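# Niagara system for the blue flame hair on the FlickLedge character head, attached to the head bone and adjusted for BigHeadMode.
# SplineThickenHair burst-spawns persistent meshes with view-angle and time-of-day tints, passing a smoothed world position and bend intensity through InterpolateDy.
# SpriteFireHair emits short-lived flipbook sprites at an owner-speed-driven rate, shaped by taper, upward motion, velocity lag and curl noise.
# User HairAlpha fades both emitters; bisHeadHidden sets their visibility tags.
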NiagaraSystem /BRCosmetics/Characters/Player/Male/Medium/Heads/M_MED_FlickLedge_Head/FX/NS_Char_FlickLedge_Hair
dependencies {
 usedDataInterfaces [
  &NiagaraDataInterfaceSkeletalMesh
  &NiagaraDataInterfaceSpriteRendererInfo
  &NiagaraDataInterfaceColorCurve
  &NiagaraDataInterfaceCurve
  &NiagaraDataInterfaceCamera
 ]
 usedDynamicInputs [
  &MakeCustomFloatFromBool
  &SetBoolByFloatComparison
  &MakeCustomVectorFromBool
 ]
 usedModules [
  &SystemState
  &self:SystemSpawnScript.SystemScriptSource.SystemScriptGraph.NiagaraNodeAssignment_8.SetVariables_#1
  &self:SystemSpawnScript.SystemScriptSource.SystemScriptGraph.NiagaraNodeAssignment_9.SetVariables_#2
 ]
}
emitters {
 SplineThickenHair {
  emitterUpdateScript [
   &EmitterState(
    Inactive Response=::NewEnumerator1
    Life Cycle Mode=::NewEnumerator1
    Loop Behavior=::NewEnumerator1
    Loop Duration Mode=::NewEnumerator1
    UseLoopDelay=false
   )
   &SpawnBurst_Instantaneous(Use Spawn Probability=false, UseLoopCountLimit=false)
  ]
  particleSpawnScript [
   &InitializeParticle(Mesh Scale Mode=::NewEnumerator3, UsePositionOffset=false, Write Material Random=true)
   &InitialMeshOrientation(Mesh Axis To Orient=<0 0 1>, Mesh Orientation Mode=::NewEnumerator2, Use Rotation Vector=true)
   &DynamicMaterialParameters(
    Index 0 Param 1=0.6
    Index 0 Param 2=1.5
    Index 0 Param 3=0.5
    Index 0 Param 4=1
    Index 2 Param 2=2
    Module.Index0Float0Write=true
    Module.Index0Float1Write=true
    Module.Index0Float2Write=true
    Module.Index0Float3Write=true
    Module.Index2Float0Write=true
    Module.Index2Float1Write=true
    Module.Index2Float2Write=true
    Module.Index2Float3Write=true
    Write Parameter Index 2=::NewEnumerator1
   )
  ]
  particleUpdateScript [
   &ParticleState(Kill Particles When Lifetime Has Elapsed=false)
   &ScaleMeshSize(Scale Factor=&Multiply_VectorByFloat(Vector=<1 1 1>, Float=&Lerp_Float(A=1.2, B=1.8, Alpha=$NPC.FortniteNPC.BigHeadMode)))
   &SkeletalMeshLocation(
    Bone Sampling Mode=::NewEnumerator4
    MaskSpawnGroup=false
    Sampled Position Offset=&Lerp_Vector(A=<8 2 0>, B=<20 3 0>, Alpha=$NPC.FortniteNPC.BigHeadMode)
    Skeletal Mesh=DataInterface() {propertyValues {
      DefaultMesh=null
      ExcludeBoneName=root
      OptionalComponentClass=null
      PreviewMesh=/BRCosmetics/Characters/Player/Male/Medium/Heads/M_MED_FlickLedge_Head/Meshes/M_MED_FlickLedge_Head
      SkinningMode=SkinOnTheFly
      SoftSourceActor=null
      SourceMode=Default
      UvSetIndex=0
      WholeMeshLOD=-1
      bExcludeBone=false
      bReadDeformedGeometry=true
      bRequireCurrentFrameData=true
      ComponentTags []
      FilteredBones [head]
      FilteredSockets []
      MeshUserParameter {parameter {name=None, typeDefHandle {registeredTypeIndex=1}, varData []}}
      SamplingRegions []
     }}
   )
   /SetVariables(
    Particles.BrightCenterColor:LinearColor=&MultiplyLinearColor(
     Linear Color=<rgba 0.0987 0.4638 0.9 1>
     Scale Factor=&MakeVector4(
      X=1
      Y=1
      Z=1
      W=&ClampFloat(Float=&RemapRange(
        Input Value To Remap=&DotProduct(
         Vector A=&NormalizeVector(
          Vector To Normalize=&ReturnCameraVector(Camera Vectors=::NewEnumerator1, Coordinate Space=::Simulation)
          Safe Normalize=true
          Safe Fallback Threshold=0
          Safe Fallback Vector=<0 0 0.001>
         )
         Vector B=&RotateVectorByQuaternion(Vector To Rotate=<0 1 0>, Quaternion=$Particles.SkeletalMeshLocation.SampledRotation)
        )
        Input Min=0.2
        Input Max=0.5
        Output Min=0
        Output Max=1
       ))
     )
    )
    Particles.EdgeColor:LinearColor=&MultiplyLinearColor(
     Linear Color=<rgba 0 0.0981 0.3281 2>
     Scale Factor=&MakeVector4_FromLinearColor(LinearColor=&FNTimeOfDayColor(
       Night Color=<rgba 0.4543 0.4862 0.8519 1>
       Morning Color=<rgba 0.9561 0.9112 0.5189 1>
       Day Color=<rgba 1 1 1 1>
       Evening Color=<rgba 0.9561 0.9387 0.6041 1>
      ))
    )
    Particles.FronLongColor:LinearColor=&MultiplyLinearColor(
     Linear Color=&MultiplyLinearColor(
      Linear Color=<rgba 0 0.0981 0.3281 2>
      Scale Factor=&MakeVector4(
       X=1
       Y=1
       Z=1
       W=&ClampFloat(Float=&RemapRange(
         Input Value To Remap=&DotProduct(
          Vector A=&NormalizeVector(
           Vector To Normalize=&ReturnCameraVector(Camera Vectors=::NewEnumerator1, Coordinate Space=::Simulation)
           Safe Normalize=true
           Safe Fallback Threshold=0
           Safe Fallback Vector=<0 0 0.001>
          )
          Vector B=&RotateVectorByQuaternion(Vector To Rotate=<0 1 0>, Quaternion=$Particles.SkeletalMeshLocation.SampledRotation)
         )
         Input Min=0.8
         Input Max=0.3
         Output Min=0
         Output Max=1
        ))
      )
     )
     Scale Factor=&MakeVector4_FromLinearColor(LinearColor=&FNTimeOfDayColor(
       Night Color=<rgba 0.4543 0.4862 0.8519 1>
       Morning Color=<rgba 1 0.9618 0.5973 1>
       Day Color=<rgba 1 1 1 1>
       Evening Color=<rgba 1 0.9772 0.6847 1>
      ))
    )
   ) {isSetParameters=true}
   /SetVariables(
    Particles.BrightCenterColor:LinearColor=&MultiplyLinearColor(
     Linear Color=$Particles.BrightCenterColor
     Scale Factor=&MakeVector4FromVectorAndFloat("Vector (RGB)"=<1 1 1>, "Float (Alpha)"=$User.HairAlpha)
    )
    Particles.Color=&MultiplyLinearColor(
     Linear Color=<rgba 0.0774 0.5446 1.05 1>
     Scale Factor=&MakeVector4FromVectorAndFloat("Vector (RGB)"=<1 1 1>, "Float (Alpha)"=$User.HairAlpha)
    )
    Particles.EdgeColor:LinearColor=&MultiplyLinearColor(
     Linear Color=$Particles.EdgeColor
     Scale Factor=&MakeVector4FromVectorAndFloat("Vector (RGB)"=<1 1 1>, "Float (Alpha)"=$User.HairAlpha)
    )
    Particles.FronLongColor:LinearColor=&MultiplyLinearColor(
     Linear Color=$Particles.FronLongColor
     Scale Factor=&MakeVector4FromVectorAndFloat("Vector (RGB)"=<1 1 1>, "Float (Alpha)"=$User.HairAlpha)
    )
   ) {isSetParameters=true}
   /SetVariables() {isSetParameters=true}
   &ScaleMeshSize(Scale Factor=<0.8 0.8 0.8>) {module=ScaleMeshSize001}
   &InterpolateOverTime(
    Rate Of Change:NiagaraFloat=$System.InterpolateRate
    Target Vector Value:Vector3f=&TransformPosition(Position=$Particles.Position, Source Space=::Simulation, Destination Space=::World)
    Variable Type=::NewEnumerator2
   )
   /SetVariables(Particles.InterpolateDy:Vector4f=&MakeFloat4FromVectorAndFloat("Vector (RGB)"=$StackContext.InterpolateOverTime.InterpolateOverTime_V3_.Moving Average, "Float (Alpha)"=$System.BendIntensity)) {isSetParameters=true}
   &CameraOffset(Camera Offset Amount=0.5)
   /SetVariables(Particles.VisibilityTag=&MakeCustomIntFromBool(Bool=$User.bisHeadHidden, True INT=1, False INT=0)) {isSetParameters=true}
  ]
  rendererClasses [&NiagaraMeshRendererProperties]
  renderers [
   {rendererClass=&NiagaraMeshRendererProperties, rendererIndex=0}
   {rendererClass=&NiagaraMeshRendererProperties, rendererIndex=1}
   {rendererClass=&NiagaraMeshRendererProperties, rendererIndex=2}
   {rendererClass=&NiagaraMeshRendererProperties, rendererIndex=3}
  ]
 }
 SpriteFireHair {
  emitterUpdateScript [
   &EmitterState(
    Inactive Response=::NewEnumerator1
    Life Cycle Mode=::NewEnumerator1
    Loop Behavior=::NewEnumerator1
    Loop Duration Mode=::NewEnumerator1
    UseLoopDelay=false
   )
   &SpawnRate(
    SpawnRate=&RemapRange(
     Input Value To Remap=&VectorLength(Vector=$Engine.Owner.Velocity)
     Input Min=75
     Input Max=150
     Output Min=25
     Output Max=60
    )
    Use Spawn Probability=false
   )
  ]
  particleSpawnScript [
   /Niagara/Modules/Spawn/Initialization/InitializeParticle(
    Lifetime=5
    Module.Write Color=true
    Module.Write Lifetime=true
    Module.Write Mass=true
    Module.Write Material Random=true
    Module.Write Position=false
    Module.Write Scale=false
    Module.Write SpriteRotation=false
    Module.Write SpriteSize=true
    Position=&SimulationPosition()
    Sprite Size=<5 5>
   ) {enabled=false}
   /Niagara/Modules/Spawn/Initialization/V2/InitializeParticle(
    Color Mode=::NewEnumerator0
    Lifetime=0.4
    Mesh Scale=<0.03 0.08 1>
    Mesh Scale Mode=::NewEnumerator4
    Sprite Rotation Angle=&RandomRangeFloat(Minimum=-0.1, Maximum=0.1)
    Sprite Rotation Mode=::NewEnumerator1
    Sprite Size Max=<9 11>
    Sprite Size Min=<7 9>
    Sprite Size Mode=::NewEnumerator2
    Sprite UV Mode=::NewEnumerator3
    UsePositionOffset=false
    Write Material Random=true
   ) {module=InitializeParticle001}
   /SetVariables(Particles.Random size:NiagaraFloat=&RandomRangeFloat(Minimum=0.5, Maximum=0.7)) {isSetParameters=true}
  ]
  particleUpdateScript [
   &ParticleState()
   &SkeletalMeshLocation(
    Bone Sampling Mode=::NewEnumerator4
    MaskSpawnGroup=false
    Orientation Sampling=::NewEnumerator1
    Sampled Position Offset=&Lerp_Vector(
     A=&AddVectorToPosition(Position=<15 -3 0>, Vector=&RandomRangeVector(Minimum=<-3 -2 -3>, Maximum=<3 2 3>))
     B=&AddVectorToPosition(Position=<34 -3 0>, Vector=&RandomRangeVector(Minimum=<-3 -3 -4>, Maximum=<3 3 4>))
     Alpha=$NPC.FortniteNPC.BigHeadMode
    )
    Skeletal Mesh=DataInterface() {propertyValues {
      DefaultMesh=null
      ExcludeBoneName=root
      OptionalComponentClass=null
      PreviewMesh=&M_MED_FlickLedge_Head
      SkinningMode=SkinOnTheFly
      SoftSourceActor=null
      SourceMode=Default
      UvSetIndex=0
      WholeMeshLOD=-1
      bExcludeBone=false
      bReadDeformedGeometry=true
      bRequireCurrentFrameData=true
      ComponentTags []
      FilteredBones [head]
      FilteredSockets []
      MeshUserParameter {parameter {name=None, typeDefHandle {registeredTypeIndex=1}, varData []}}
      SamplingRegions []
     }}
   )
   &self:TaperToPoint(AdjBonOrientOnLocalZAxis=-7) {module=TaperToPoint}
   &self:UpMotion(LerpLocalToWorldSpace=0.4, UpSpeed=20) {module=UpMotion}
   &self:LagParticlePositionWithVel(JogLag=0.05, JogSpeed=300, SkyDiveLag=0.01, SkyDiveSpeed=700) {module=LagParticlePositionWithVel}
   &CurlNoiseForce(
    Bias Noise=false
    Noise Frequency=75
    Noise Strength=&Multiply_Float(A=500, B=$Particles.NormalizedAge)
    Pan Noise=true
    Pan Noise Field=<0 0 1>
   )
   &SubUVAnimation(
    End Frame Range Override=3
    Evaluation Type=::NewEnumerator0
    "Sprite Renderer "=DataInterface() {propertyValues {SpriteRenderer=&self:SpriteFireHair_0.NiagaraSpriteRendererProperties_1}}
    SubUV Animation Mode:ENiagara_SubUVLookupModeV2=::NewEnumerator2
    UseEndFrame=true
    UseStartFrame=true
   )
   &SolveForcesAndVelocity(Clamp Velocity=false, Limit Acceleration=false)
   &Color(
    Color=&ColorFromCurve(ColorCurve=DataInterface() {propertyValues {
       AlphaCurve {keys [<at 0 0 RCIM_Cubic>,<at 0.2471 2>,<at 0.5886 2>,<at 1 0>]}
       BlueCurve {keys [<at 0.5879 0.9>,<at 1 0.1496>]}
       GreenCurve {keys [<at 0.5879 0.4104>,<at 1 0.0827>]}
       RedCurve {keys [<at 0.5879 0.0845>,<at 1 0>]}
      }})
    Scale Alpha=$User.HairAlpha
   )
   &ScaleSpriteSize(Uniform Curve Sprite Scale=DataInterface() {propertyValues {Curve {keys [
        {arriveTangent=2.652, interpMode=RCIM_Cubic, time=0, value=0.75}
        {arriveTangent=0.4737, interpMode=RCIM_Cubic, leaveTangent=0.4737, time=0.3, value=1.0816}
        {arriveTangent=-1.2594, interpMode=RCIM_Cubic, leaveTangent=-1.2594, time=0.7, value=1.0816}
        {arriveTangentWeight=0.1011, interpMode=RCIM_Cubic, tangentMode=RCTM_User, time=1, value=0.2}
       ]}}})
   /SetVariables(
    Particles.CoreInFrontColor:LinearColor=&Multiply_LinearColorByFloat(
     Linear Color=&MultiplyLinearColor(
      Linear Color=$Particles.Color
      Scale Factor=&MakeVector4(
       X=1
       Y=1
       Z=1
       W=&ClampFloat(Float=&RemapRange(
         Input Value To Remap=&DotProduct(
          Vector A=&NormalizeVector(
           Vector To Normalize=&ReturnCameraVector(Camera Vectors=::NewEnumerator1, Coordinate Space=::Simulation)
           Safe Normalize=true
           Safe Fallback Threshold=0
           Safe Fallback Vector=<0 0 0.001>
          )
          Vector B=&RotateVectorByQuaternion(Vector To Rotate=<-1 0 0>, Quaternion=$Particles.SkeletalMeshLocation.SampledRotation)
         )
         Input Min=0.3
         Input Max=0.6
         Output Min=0
         Output Max=1
        ))
      )
     )
     Float=1
    )
    Particles.GlowColor=&MultiplyLinearColor(Linear Color=$Particles.Color, Scale Factor=<1 1 0.6 0.15>)
    Particles.GlowScale:Vector2f=&Multiply_Vector2D_ByFloat(Float=1.4, Vector2D=$Particles.SpriteSize)
   ) {isSetParameters=true}
   /SetVariables(Particles.VisibilityTag=&MakeCustomIntFromBool(Bool=$User.bisHeadHidden, True INT=1, False INT=0)) {isSetParameters=true}
  ]
  rendererClasses [&NiagaraSpriteRendererProperties]
  renderers [
   {rendererClass=&NiagaraSpriteRendererProperties}
   {rendererClass=&NiagaraSpriteRendererProperties, rendererIndex=1}
   {rendererClass=&NiagaraSpriteRendererProperties, rendererIndex=2}
   {rendererClass=&NiagaraSpriteRendererProperties, rendererIndex=3}
  ]
 }
}
summary {
 systemName=NS_Char_FlickLedge_Hair
 emitters [
  {emitterName=SpriteFireHair, rendererClasses [&NiagaraSpriteRendererProperties]}
  {emitterName=SplineThickenHair, rendererClasses [&NiagaraMeshRendererProperties]}
 ]
 userVariables [
  User.bIsFrontEnd {type=NiagaraBool}
  User.bIsFrontEndPreview {type=NiagaraBool}
  User.bisHeadHidden {type=NiagaraBool}
  User.bisSwinging {type=NiagaraBool}
  User.Emote begin {type=NiagaraBool}
  User.HairAlpha {type=NiagaraFloat, defaultValue=1, description=""}
  User.IsSkydiving {type=NiagaraBool}
 ]
}
